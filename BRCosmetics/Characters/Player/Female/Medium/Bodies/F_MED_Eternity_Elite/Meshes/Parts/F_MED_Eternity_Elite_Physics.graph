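# Physics asset for the F_MED_Eternity_Elite female character body, setting up jointed capsule chains for the ponytail and front hair strands.
# Tiny MSTR bodies link each chain to the head, with swing and narrow twist limits down the links and extra strands branching from the lower ponytail.
# Head, spine, pelvis, upper-arm and backpack capsules form the surrounding body shapes, joined by angularly free constraints.
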PhysicsAsset /BRCosmetics/Characters/Player/Female/Medium/Bodies/F_MED_Eternity_Elite/Meshes/Parts/F_MED_Eternity_Elite_Physics
bodies {
 dyn_backpack_COL [
  {shapeType=Capsule, length=17.5, radius=9.755, rotation=<rot 0 -90 0>}
  {shapeType=Capsule, length=20.75, radius=9.25, center=<18.744 0.0003 -1.2395>}
 ]
 dyn_hairstrand_fr_01_l [{shapeType=Capsule, length=4.3527, radius=1.7725, center=<2.2794 0 0>, rotation=<rot -90 -0.0264 0.0264>}]
 dyn_hairstrand_fr_01_l_MSTR [{shapeType=Capsule, length=0.01, radius=0.01}]
 dyn_hairstrand_fr_01_r [{shapeType=Capsule, length=4.3527, radius=1.7725, center=<2.2794 0 0>, rotation=<rot -90 -0.0264 0.0264>}]
 dyn_hairstrand_fr_01_r_MSTR [{shapeType=Capsule, length=0.01, radius=0.01}]
 dyn_hairstrand_fr_02_l [{shapeType=Capsule, length=4.6027, radius=1.5225, center=<2.2794 0 0>, rotation=<rot -90 -0.0264 0.0264>}]
 dyn_hairstrand_fr_02_r [{shapeType=Capsule, length=4.3527, radius=1.7725, center=<2.2794 0 0>, rotation=<rot -90 -0.0264 0.0264>}]
 dyn_ponytail_01 [{shapeType=Capsule, length=3.1027, radius=3.0225, center=<2.2794 0 0>, rotation=<rot -90 -0.0264 0.0264>}]
 dyn_ponytail_01_MSTR [{shapeType=Capsule, length=0.01, radius=0.01}]
 dyn_ponytail_02 [{shapeType=Capsule, length=1.8527, radius=4.2725, center=<3.8497 0 0>, rotation=<rot -90 0.0396 -0.0396>}]
 dyn_ponytail_03 [{shapeType=Capsule, length=9, radius=5.7725, center=<5.8182 0 0>, rotation=<rot -90 -0.0198 0.0198>}]
 dyn_ponytail_04 [{shapeType=Capsule, length=9, radius=5.7725, center=<5.8182 0 0>, rotation=<rot -90 -0.0198 0.0198>}]
 dyn_ponytail_05 [{shapeType=Capsule, length=10.75, radius=4.0225, center=<5.8182 0 0>, rotation=<rot -90 -0.0198 0.0198>}]
 dyn_ponytail_06 [{shapeType=Capsule, length=12, radius=2.7725, center=<5.8182 0 0>, rotation=<rot -90 -0.0198 0.0198>}]
 dyn_ponytailstrand_01 [{shapeType=Capsule, length=7.75, radius=3.2725, center=<3.2579 -0.0006 0>, rotation=<rot -90 -0.0198 0.0198>}]
 dyn_ponytailstrand_02 [{shapeType=Capsule, length=7.75, radius=3.2725, center=<3.2579 -0.0006 0>, rotation=<rot -90 -0.0198 0.0198>}]
 dyn_ponytailstrand_03 [{shapeType=Capsule, length=17.5, radius=3.0225, center=<8.7802 -0.0007 0>, rotation=<rot -90 179.9802 -179.9802>}]
 dyn_ponytailstrand_04 [{shapeType=Capsule, length=7.75, radius=3.2725, center=<3.2579 -0.0006 0>, rotation=<rot -90 -0.0198 0.0198>}]
 dyn_ponytailstrand_05 [{shapeType=Capsule, length=7.75, radius=3.2725, center=<3.2579 -0.0006 0>, rotation=<rot -90 -0.0198 0.0198>}]
 head [{shapeType=Capsule, length=6.2659, radius=7.3573, center=<3.4291 0.3722 0>, rotation=<rot -90 0.0132 -0.0131>}]
 pelvis [{shapeType=Capsule, length=16.7659, radius=10.3573, center=<-1.3262 -1.4944 0.0001>, rotation=<rot 0.0006 -179.9999 179.9738>}]
 spine_03 [{shapeType=Capsule, length=7.7659, radius=11.6073, center=<-2.095 5.145 -0.0003>, rotation=<rot 0.0006 -179.9999 180>}]
 spine_05 [{shapeType=Capsule, length=15.7659, radius=9.3573, center=<0.429 1.3142 0.0002>, rotation=<rot 0.0006 -179.9999 180>}]
 upperarm_l [{shapeType=Capsule, length=25.5159, radius=6.1073, center=<14.9673 0.3723 0.0001>, rotation=<rot 90 -0.0201 -0.0204>}]
 upperarm_r [{shapeType=Capsule, length=25.5159, radius=6.1073, center=<-14.9673 -0.3723 -0.0001>, rotation=<rot 90 0.0195 -179.9808>}]
}
constraints [
 {
  bone1Name=dyn_ponytail_01_MSTR
  bone2Name=head
  swing1LimitDegrees=45
  swing1Motion=Free
  swing2LimitDegrees=45
  swing2Motion=Free
  twistLimitDegrees=45
  twistMotion=Free
 }
 {
  bone1Name=dyn_hairstrand_fr_01_l_MSTR
  bone2Name=head
  swing1LimitDegrees=45
  swing1Motion=Free
  swing2LimitDegrees=45
  swing2Motion=Free
  twistLimitDegrees=45
  twistMotion=Free
 }
 {
  bone1Name=dyn_hairstrand_fr_01_r_MSTR
  bone2Name=head
  swing1LimitDegrees=45
  swing1Motion=Free
  swing2LimitDegrees=45
  swing2Motion=Free
  twistLimitDegrees=45
  twistMotion=Free
 }
 {bone1Name=dyn_ponytail_01, bone2Name=dyn_ponytail_01_MSTR, swing1LimitDegrees=15, swing2LimitDegrees=15, twistLimitDegrees=5}
 {bone1Name=dyn_ponytail_02, bone2Name=dyn_ponytail_01, swing1LimitDegrees=15, swing2LimitDegrees=15, twistLimitDegrees=5}
 {bone1Name=dyn_ponytail_03, bone2Name=dyn_ponytail_02, swing1LimitDegrees=45, swing2LimitDegrees=45, twistLimitDegrees=5}
 {bone1Name=dyn_ponytail_04, bone2Name=dyn_ponytail_03, swing1LimitDegrees=45, swing2LimitDegrees=45, twistLimitDegrees=5}
 {bone1Name=dyn_ponytail_05, bone2Name=dyn_ponytail_04, swing1LimitDegrees=45, swing2LimitDegrees=45, twistLimitDegrees=5}
 {bone1Name=dyn_ponytail_06, bone2Name=dyn_ponytail_05, swing1LimitDegrees=45, swing2LimitDegrees=45, twistLimitDegrees=5}
 {bone1Name=dyn_hairstrand_fr_01_l, bone2Name=dyn_hairstrand_fr_01_l_MSTR, swing1LimitDegrees=45, swing2LimitDegrees=25, twistLimitDegrees=5}
 {bone1Name=dyn_hairstrand_fr_02_l, bone2Name=dyn_hairstrand_fr_01_l, swing1LimitDegrees=75, swing2LimitDegrees=45, twistLimitDegrees=5}
 {bone1Name=dyn_hairstrand_fr_01_r, bone2Name=dyn_hairstrand_fr_01_r_MSTR, swing1LimitDegrees=45, swing2LimitDegrees=25, twistLimitDegrees=5}
 {bone1Name=dyn_hairstrand_fr_02_r, bone2Name=dyn_hairstrand_fr_01_r, swing1LimitDegrees=75, swing2LimitDegrees=45, twistLimitDegrees=5}
 {
  bone1Name=head
  bone2Name=spine_05
  swing1LimitDegrees=45
  swing1Motion=Free
  swing2LimitDegrees=45
  swing2Motion=Free
  twistLimitDegrees=45
  twistMotion=Free
 }
 {
  bone1Name=spine_05
  bone2Name=spine_03
  swing1LimitDegrees=45
  swing1Motion=Free
  swing2LimitDegrees=45
  swing2Motion=Free
  twistLimitDegrees=45
  twistMotion=Free
 }
 {
  bone1Name=spine_03
  bone2Name=pelvis
  swing1LimitDegrees=45
  swing1Motion=Free
  swing2LimitDegrees=45
  swing2Motion=Free
  twistLimitDegrees=45
  twistMotion=Free
 }
 {bone1Name=dyn_ponytailstrand_05, bone2Name=dyn_ponytail_06, swing1LimitDegrees=45, swing2LimitDegrees=45, twistLimitDegrees=5}
 {bone1Name=dyn_ponytailstrand_04, bone2Name=dyn_ponytail_06, swing1LimitDegrees=45, swing2LimitDegrees=45, twistLimitDegrees=5}
 {bone1Name=dyn_ponytailstrand_02, bone2Name=dyn_ponytail_05, swing1LimitDegrees=45, swing2LimitDegrees=45, twistLimitDegrees=5}
 {bone1Name=dyn_ponytailstrand_03, bone2Name=dyn_ponytail_05, swing1LimitDegrees=5, swing2LimitDegrees=45, twistLimitDegrees=5}
 {bone1Name=dyn_ponytailstrand_01, bone2Name=dyn_ponytail_05, swing1LimitDegrees=45, swing2LimitDegrees=45, twistLimitDegrees=5}
 {
  bone1Name=upperarm_l
  bone2Name=spine_05
  swing1LimitDegrees=45
  swing1Motion=Free
  swing2LimitDegrees=45
  swing2Motion=Free
  twistLimitDegrees=45
  twistMotion=Free
 }
 {
  bone1Name=upperarm_r
  bone2Name=spine_05
  swing1LimitDegrees=45
  swing1Motion=Free
  swing2LimitDegrees=45
  swing2Motion=Free
  twistLimitDegrees=45
  twistMotion=Free
 }
 {
  bone1Name=dyn_backpack_COL
  bone2Name=spine_05
  swing1LimitDegrees=45
  swing1Motion=Free
  swing2LimitDegrees=45
  swing2Motion=Free
  twistLimitDegrees=45
  twistMotion=Free
 }
]
