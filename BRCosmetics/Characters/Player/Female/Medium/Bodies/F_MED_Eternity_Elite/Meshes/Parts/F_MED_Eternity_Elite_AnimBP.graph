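# Anim Blueprint for the F_MED_Eternity_Elite character part, adding physics-driven ponytail motion.
# It copies attached-parent and Body_Mesh poses, blended at the arms, into a spine_05-space RigidBody whose gravity follows skydiving force.
# Backbling and cape flags, set on customization load, offset ponytail and backpack-collision bones around the simulation.
# A pelvis Look At replaces disabled physics; constraint and noise-driven strand chains sit disconnected from the output.
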AnimBlueprint /BRCosmetics/Characters/Player/Female/Medium/Bodies/F_MED_Eternity_Elite/Meshes/Parts/F_MED_Eternity_Elite_AnimBP
target_skeleton=&/F_MED_Eternity_Elite_Skeleton
graphs [
 AnimGraph {nodes [
   n1 = "Constraint - dyn_ponytail_01 (P)"(Alpha=1, ConstraintWeights_0=1) {
    alphaBoolBlend {blendOption=Linear}
    boneToModify {boneName=dyn_ponytail_01}
    constraintSetup [{targetBone {boneName=head}}]
    constraintWeights [1]
   }
   n2 = "Constraint - dyn_ponytail_02 (P)"(ComponentPose=@n1, Alpha=1, ConstraintWeights_0=1) {
    alphaBoolBlend {blendOption=Linear}
    boneToModify {boneName=dyn_ponytail_02}
    constraintSetup [{targetBone {boneName=head}}]
    constraintWeights [1]
   }
   n3 = "Constraint - dyn_ponytail_03 (P)"(ComponentPose=@n2, Alpha=1, ConstraintWeights_0=1) {
    alphaBoolBlend {blendOption=Linear}
    boneToModify {boneName=dyn_ponytail_03}
    constraintSetup [{targetBone {boneName=neck_02}}]
    constraintWeights [1]
   }
   n4 = "Constraint - dyn_ponytail_04 (P)"(ComponentPose=@n3, Alpha=1, ConstraintWeights_0=1) {
    alphaBoolBlend {blendOption=Linear}
    boneToModify {boneName=dyn_ponytail_04}
    constraintSetup [{targetBone {boneName=spine_05}}]
    constraintWeights [1]
   }
   n5 = "Constraint - dyn_ponytail_05 (P)"(ComponentPose=@n4, Alpha=1, ConstraintWeights_0=1) {
    alphaBoolBlend {blendOption=Linear}
    boneToModify {boneName=dyn_ponytail_05}
    constraintSetup [{targetBone {boneName=spine_04}}]
    constraintWeights [1]
   }
   "Constraint - dyn_ponytail_06 (P)"(ComponentPose=@n5, Alpha=1, ConstraintWeights_0=1) {
    alphaBoolBlend {blendOption=Linear}
    boneToModify {boneName=dyn_ponytail_06}
    constraintSetup [{targetBone {boneName=spine_03}}]
    constraintWeights [1]
   }
   n6 = Get extraRotation() -> (extraRotation:Rotator)
   n7 = "Transform (Modify) Bone - Bone: dyn_ponytailstrand_02"(Alpha=0.25, Translation="0,0,0", Rotation=@n6, Scale="1,1,1") {
    rotationMode=BMM_Additive
    rotationSpace=BCS_BoneSpace
    scaleMode=BMM_Additive
    scaleSpace=BCS_BoneSpace
    translationMode=BMM_Additive
    translationSpace=BCS_BoneSpace
    alphaBoolBlend {blendOption=Linear}
    boneToModify {boneName=dyn_ponytailstrand_02}
   }
   n8 = "Transform (Modify) Bone - Bone: dyn_ponytailstrand_03"(ComponentPose=@n7, Alpha=0.2, Translation="0,0,0", Rotation=@n6, Scale="1,1,1") {
    rotationMode=BMM_Additive
    rotationSpace=BCS_BoneSpace
    scaleMode=BMM_Additive
    scaleSpace=BCS_BoneSpace
    translationMode=BMM_Additive
    translationSpace=BCS_BoneSpace
    alphaBoolBlend {blendOption=Linear}
    boneToModify {boneName=dyn_ponytailstrand_03}
   }
   n9 = "Transform (Modify) Bone - Bone: dyn_ponytailstrand_01"(ComponentPose=@n8, Alpha=0.5, Translation="0,0,0", Rotation=@n6, Scale="1,1,1") {
    rotationMode=BMM_Additive
    rotationSpace=BCS_BoneSpace
    scaleMode=BMM_Additive
    scaleSpace=BCS_BoneSpace
    translationMode=BMM_Additive
    translationSpace=BCS_BoneSpace
    alphaBoolBlend {blendOption=Linear}
    boneToModify {boneName=dyn_ponytailstrand_01}
   }
   n10 = "Transform (Modify) Bone - Bone: dyn_ponytailstrand_04"(ComponentPose=@n9, Alpha=0.5, Translation="0,0,0", Rotation=@n6, Scale="1,1,1") {
    rotationMode=BMM_Additive
    rotationSpace=BCS_BoneSpace
    scaleMode=BMM_Additive
    scaleSpace=BCS_BoneSpace
    translationMode=BMM_Additive
    translationSpace=BCS_BoneSpace
    alphaBoolBlend {blendOption=Linear}
    boneToModify {boneName=dyn_ponytailstrand_04}
   }
   "Transform (Modify) Bone - Bone: dyn_ponytailstrand_05"(ComponentPose=@n10, Alpha=0.5, Translation="0,0,0", Rotation=@n6, Scale="1,1,1") {
    rotationMode=BMM_Additive
    rotationSpace=BCS_BoneSpace
    scaleMode=BMM_Additive
    scaleSpace=BCS_BoneSpace
    translationMode=BMM_Additive
    translationSpace=BCS_BoneSpace
    alphaBoolBlend {blendOption=Linear}
    boneToModify {boneName=dyn_ponytailstrand_05}
   }
   n11 = Copy Pose From Mesh(SourceMeshComponent=Get Body_Mesh())
   n12 = Copy Pose From Mesh() {bUseAttachedParent=true}
   n13 = Get isWearingBackbling() -> (isWearingBackbling:double)
   n14 = "Transform (Modify) Bone - Bone: dyn_ponytail_01_MSTR"(
    ComponentPose=Local To Component(LocalPose=Layered blend per bone(BasePose=@n12, BlendPoses_0=@n11) {layerSetup [{branchFilters [{boneName=upperarm_l}, {boneName=upperarm_r}]}]})
    Alpha=@n13
    Translation="0,0,0"
    Rotation="0.000,0.000,20.000"
    Scale="1,1,1"
   ) {
    rotationMode=BMM_Additive
    rotationSpace=BCS_BoneSpace
    translationSpace=BCS_BoneSpace
    alphaBoolBlend {blendOption=Linear}
    alphaScaleBiasClamp {bInterpResult=true}
    boneToModify {boneName=dyn_ponytail_01_MSTR}
    rotation {roll=20}
   }
   n15 = "Transform (Modify) Bone - Bone: dyn_backpack_COL"(ComponentPose=@n14, Alpha=@n13, Translation="39.912,-0.000,0.000", Rotation="0,0,0", Scale="1,1,1") {
    rotationMode=BMM_Additive
    rotationSpace=BCS_BoneSpace
    scaleMode=BMM_Additive
    scaleSpace=BCS_BoneSpace
    translationMode=BMM_Additive
    translationSpace=BCS_BoneSpace
    alphaBoolBlend {blendOption=Linear}
    boneToModify {boneName=dyn_backpack_COL}
    translation {x=39.912}
   }
   n16 = RigidBody!(ComponentPose=@n15, Alpha, OverrideWorldGravity:Vector=Get gravityVector() -> (gravityVector:Vector), ExternalForce) -> (Pose) {
    bClampLinearTranslationLimitToRefPose=true
    bOverrideWorldGravity=true
    overridePhysicsAsset=&/F_MED_Eternity_Elite_Physics
    simulationSpace=BaseBoneSpace
    alphaBoolBlend {blendOption=Linear}
    baseBoneRef {boneName=spine_05}
    componentAppliedLinearAccClamp=<1000 1000 1200>
    componentLinearAccScale=<0.1 0.1 0.5>
    componentLinearVelScale=<0.4 0.4 0.5>
    overrideWorldGravity {z=-980}
   }
   n17 = "Transform (Modify) Bone - Bone: dyn_ponytail_01_MSTR"(ComponentPose=@n16, Alpha=Get isWearingCape() -> (isWearingCape:double), Translation="0,0,0", Rotation="5,0.000,0.000", Scale="1,1,1") {
    rotationMode=BMM_Additive
    rotationSpace=BCS_BoneSpace
    translationSpace=BCS_BoneSpace
    alphaBoolBlend {blendOption=Linear}
    alphaScaleBiasClamp {bInterpResult=true}
    boneToModify {boneName=dyn_ponytail_01_MSTR}
    rotation {pitch=5}
   }
   n18 = Look At - Bone: dyn_ponytail_03!(ComponentPose=@n17, bAlphaBoolEnabled:bool=NOT Boolean(A=Get bRigidBodyEnabled())) -> (Pose) {
    alphaInputType=Bool
    alphaBoolBlend {blendOption=Linear}
    boneToModify {boneName=dyn_ponytail_03}
    lookAtLocation=<0 -15>
    lookAtTarget {boneReference {boneName=pelvis}}
    lookAt_Axis {axis=<1 0>}
   }
   Output Pose(Result=Component To Local(ComponentPose=@n18)) {layerGroup=None}
   Get simSpaceSettings() -> (simSpaceSettings:Sim Space Settings Structure)
  ]}
 EventGraph {nodes [
   Event BlueprintUpdateAnimation()
   > Sequence!(execute) -> (then_0, then_1, then_2:Exec, then_3:Exec)
   n1 = Break Fort Anim Input Skydiving External Force(FortAnimInput_SkydivingExternalForce=Get SkydivingExternalForceInput()) -> (
    bUseSkydivingVectorForce=~
    SkydivingVectorForce=~
    HeadToPelvisDirection=~
    FloatingMultiplier=~
    FloatingAdditive=~
    DivingMultiplier=~
    DivingAdditive=~
    ParachutingMultiplier=~
    ParachutingAdditive=~
   )
   > Set gravityVector(gravityVector:Vector=SelectVector(A=vector * float(A=@n1.SkydivingVectorForce, B=0.15), B="0, 0,-980", bPickA=Get bIsSkyDiving())) -> (Output_Get:Vector)
   Event OnCharacterCustomizationLoaded()
   > Do Once()
   n2 = CheckForBackbling()
   n3 = NOT Boolean(A=@n2.bIsBackblingLong)
   > Set isWearingBackbling(isWearingBackbling:double="To Float (Boolean)"(InBool=AND Boolean(A=@n2, B=@n3))) -> (Output_Get:double=0)
   > Set isWearingCape(isWearingCape:double="To Float (Boolean)"(InBool=@n2.bIsBackblingLong)) -> (Output_Get:double)
   Event BlueprintInitializeAnimation()
   > Set Body_Mesh(Body_Mesh=GetOwnerSkeletalMeshComponent())
   n4 = GetOwnerSkeletalMeshComponent()
   n5 = GetSocketLocation!(self:Scene Component Object Reference=@n4, InSocketName:Name=pelvis) -> (ReturnValue:Vector)
   n6 = Make Perlin Noise Vector and Remap!(
    X:float=Make Perlin Noise and Remap!(Value:float=GetGameTimeInSeconds(), RangeOutMin:float=0, RangeOutMax:float=5) -> (ReturnValue:float)
    Y:float=0
    Z:float=0
    RangeOutMinX:float=200
    RangeOutMaxX:float=500
    RangeOutMinY:float=0
    RangeOutMaxY:float=0
    RangeOutMinZ:float=0
    RangeOutMaxZ:float=0
   ) -> (ReturnValue:Vector)
   n7 = vector * vector(
    A:Vector="Get Unit Direction (Vector)"!(From:Vector=@n5, To:Vector=GetSocketLocation(self=@n4, InSocketName=thigh_r)) -> (ReturnValue:Vector) {is=title K2Node_CallFunction}
    B:Vector=@n6
   ) -> (ReturnValue:Vector) {is=title K2Node_CommutativeAssociativeBinaryOperator}
   n8 = SelectFloat(A=0, B=1, bPickA=Get bIsTargeting())
   n9 = Get PawnSpeed2D()
   Set simSpaceSettings(simSpaceSettings:Sim Space Settings Structure=Make Sim Space Settings!(
     WorldAlpha:float=MapRangeClamped(Value=@n9, InRangeA=0, InRangeB=250, OutRangeA=@n8, OutRangeB=0)
     VelocityScaleZ:float=0.1
     DampingAlpha:float=1
     MaxLinearVelocity:float=1000
     MaxAngularVelocity:float=500
     MaxLinearAcceleration:float=1000
     MaxAngularAcceleration:float=500
     ExternalLinearDragV:Vector="3,3,3"
     ExternalLinearVelocity:Vector=vector * float(A=@n7, B=0.8)
     ExternalAngularVelocity:Vector="0,0,0"
    ) -> (SimSpaceSettings:Sim Space Settings Structure) {is=title K2Node_MakeStruct}) -> (Output_Get:Sim Space Settings Structure)
   n10 = AND Boolean(
    A="float <= float"!(A:double=Get PawnSpeed2D(), B:double=100) -> (ReturnValue:bool) {is=title K2Node_CallFunction}
    B=NOT Boolean(A=Get bIsTargeting())
   )
   n11 = "Lerp (Vector)"!(A:Vector="0, 0, 0", B:Vector=vector * float(A=vector * float(A=@n6, B=0.6), B=SelectFloat(A=1, B=0, bPickA=@n10)), Alpha:float=0.5) -> (ReturnValue:Vector)
   > Set extraRotation(extraRotation:Rotator=MakeRotator(Pitch=BreakVector(InVec=@n11))) -> (Output_Get:Rotator)
   n12 = BreakVector(InVec=@n7)
   DrawDebugCoordinateSystem!(
    execute
    AxisLoc:Vector=@n5
    AxisRot:Rotator=MakeRotator(Roll=@n12, Pitch=@n12.Y, Yaw=@n12.Z)
    Scale:float=100
    Duration:float=0
    Thickness:float=1
    DepthPriority:EDrawDebugSceneDepthPriorityGroup Enum=World
   ) -> (then)
  ]}
]
variables [
 B {type=number}
 Body_Mesh {type=SkeletalMeshComponent}
 extraRotation {type=Rotator}
 gravityVector {type=Vector}
 isWearingBackbling {type=number}
 isWearingCape {type=number}
 NewVar_0 {type=Vector}
 NewVar_1 {type=number}
 NewVar_2 {type=number}
 NewVar_3 {type=number}
 NewVar_4 {
  type=SimSpaceSettings
  default {
   dampingAlpha=1
   maxAngularAcceleration=10000
   maxAngularVelocity=10000
   maxLinearAcceleration=10000
   maxLinearVelocity=10000
   velocityScaleZ=1
   worldAlpha=0
   externalLinearDragV=<0 0 0>
  }
 }
 NewVar_5 {type=AnimInstance, default=null}
 NewVar_6 {type=AnimMontage, default=null}
 NewVar_7 {type=Vector}
 simSpaceSettings {
  type=SimSpaceSettings
  default {
   dampingAlpha=1
   maxAngularAcceleration=1000
   maxAngularVelocity=1000
   maxLinearAcceleration=1000
   maxLinearVelocity=1000
   velocityScaleZ=0.1
   worldAlpha=1
   externalLinearDragV=<10 2 2>
   externalLinearVelocity=<50 2 2>
  }
 }
 wind {type=Vector}
 windAlphaDown {type=number}
 windAlphaUp {type=number}
]
